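# Translucent, unlit, two-sided mesh-particle effect material for the GoatRobe backpack, tinted by particle color.
# It draws procedural UV1 edge and scrolling stripe lines, or, via static switches, a plain fill or panning symbol texture.
# Emissive gets a mobile clamp and boost, while opacity passes through a time-of-day function fed one value for all periods.
# World position offset uses either SplineThicken or vertex-red-masked widening and directional stretch.
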Material /BRCosmetics/Accessories/FORT_Backpacks/M_MED_GoatRobe_Backpack/M_Backpack_GoatRobe
graph {
 properties {
  bEnableResponsiveAA=true
  bUsedWithMeshParticles=true
  blendMode=BLEND_Translucent
  pixelDepthOffsetMode=PDOM_Legacy
  shadingModel=MSM_Unlit
  twoSided=true
  usedShadingModels=Unlit
 }
 nodes [
  n1 = TextureCoordinate() {coordinateIndex=1}
  n2 = TextureSample!(
   UVs=Panner(Coordinate=MaterialFunctionCall(XY=@n1) {materialFunction=&Swizzle}.YX) {speedX=0.038}
   Tex
   Level
   Apply View MipBias
  ) {
   mipValueMode=TMVM_MipLevel
   samplerType=SAMPLERTYPE_Masks
   texture=&/T_GoatRobe_Symbol
  }
  n3 = ParticleColor()
  n4 = MaterialFunctionCall(Float2=TextureCoordinate() {coordinateIndex=1}) {materialFunction=&BreakOutFloat2Components}
  n5 = Multiply(A=@n4.R) {constB=0.5}
  n6 = Saturate(None=Multiply(A=Sine(None=@n5), B=ScalarParameter() {defaultValue=3, parameterName=OutterLine_Thickness}) {constB=3})
  n7 = Divide(A=Time(), B=ScalarParameter() {defaultValue=-30, parameterName=RotateSpeed}) {constB=-30}
  n8 = Frac(None=Multiply(A=Add(A=@n7, B=@n4.G)) {constB=50})
  n9 = Round(None=Saturate(None=Multiply(A=@n8) {constB=0.6}))
  n10 = ScalarParameter() {defaultValue=1, parameterName=Detail_Thin_HorizontalLine}
  n11 = Divide(A=@n7)
  n12 = Frac(None=Multiply(A=Add(A=@n11, B=@n4.G)) {constB=15})
  n13 = Round(None=Saturate(None=Multiply(A=@n12) {constB=0.55}))
  n14 = ScalarParameter() {defaultValue=1, parameterName=Detail_Thick_HorizontalLine}
  n15 = Add(A=Round(None=OneMinus(None=@n6)), B=Add(A=Multiply(A=@n9, B=@n10), B=Multiply(A=@n13, B=@n14)))
  n16 = StaticSwitchParameter(True=StaticSwitchParameter(True=Multiply(A=@n2.R, B=@n3.RGB), False=@n3.RGB) {parameterName=AddSymbol}, False=Multiply(A=@n3.RGB, B=@n15)) {parameterName=NoPattern}
  n17 = Saturate(None=@n16)
  n18 = ScalarParameter() {defaultValue=2.5, parameterName=MobilePlatformEmissveOverride}
  MaterialFunctionCall(Emissive=ShadingPathSwitch(Default=@n16, Mobile=Multiply(A=@n17, B=@n18)) {desc=Migrated from FeatureLevelSwitch}) {materialFunction=&MF_FX_EmissiveAdjustments}
  n19 = Multiply(A=@n2.R, B=@n3.A)
  n20 = VectorParameter() {parameterName=TimeOfDay_Opacity, defaultValue=<rgba 0.25 0.8 0.45 0.25>}
  Multiply(
   A=StaticSwitchParameter(True=StaticSwitchParameter(True=@n19, False=@n3.A) {parameterName=AddSymbol}, False=Multiply(A=@n3.A, B=@n15)) {parameterName=NoPattern}
   B=MaterialFunctionCall(Color_Morning=@n20.R, Color_Day=@n20.R, Color_Evening=@n20.R, Color_Night=@n20.R) {materialFunction=&Color_Time_of_Day}.Result
  )
  ScalarParameter() {defaultValue=5, parameterName=SplineThickness}
  n21 = DynamicParameter() {paramNames [SplineThickness, Param2, Param3, Param4]}
  n22 = VertexColor()
  n23 = Multiply(A=Subtract(A=WorldPosition().XYZ, B=ObjectPositionWS()), B=ScalarParameter() {defaultValue=1, parameterName=Width})
  n24 = Transform(None=VectorParameter() {parameterName=StretchDirection, defaultValue=<rgba 0 0 0 1>}.RGB) {transformSourceType=TRANSFORMSOURCE_Instance}
  n25 = VertexColor()
  StaticSwitchParameter(
   True=MaterialFunctionCall(WidthBase=@n21.SplineThickness, WidthTip=@n21.SplineThickness, Expand U or V UV Channel=StaticBool()) {materialFunction=&SplineThicken}.WorldPositionOffset
   False=Add(A=Multiply(A=@n22.R, B=@n23), B=Multiply(A=@n24, B=@n25.R))
  ) {parameterName=UseSplineMeshWPO}
 ]
}
