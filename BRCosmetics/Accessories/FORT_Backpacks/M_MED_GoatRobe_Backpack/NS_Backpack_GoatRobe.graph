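# Niagara system for the GoatRobe backpack cosmetic that draws a static, glowing magenta-pink emblem with a light.
# TriangleOutter and TriangleInner each burst three mesh particles, using partition-indexed arrays to place and rotate triangle edges around System.Position.
# Symbol feeds one particle to multiple mesh renderers through per-mesh attribute modules, while Light adds a sprite glow and light, dimmed in the front end.
# Particles loop their lifetime instead of dying, so the emblem persists.
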NiagaraSystem /BRCosmetics/Accessories/FORT_Backpacks/M_MED_GoatRobe_Backpack/NS_Backpack_GoatRobe
dependencies {
 usedDataInterfaces [&NiagaraDataInterfaceArrayFloat3, &NiagaraDataInterfaceVectorCurve]
 usedDynamicInputs []
 usedModules [
  &self:SystemSpawnScript.SystemScriptSource.SystemScriptGraph.NiagaraNodeAssignment_0.SetVariables_#1
  &SystemState
 ]
}
emitters {
 Light {
  emitterUpdateScript [
   &EmitterState(
    Inactive Response=::NewEnumerator1
    Life Cycle Mode=::NewEnumerator1
    Loop Behavior=::NewEnumerator1
    Loop Duration Mode=::NewEnumerator1
    UseLoopDelay=false
   )
   &SpawnBurst_Instantaneous(Use Spawn Probability=false, UseLoopCountLimit=false)
  ]
  particleSpawnScript [
   &InitializeParticle(
    Lifetime=4
    Position Offset=&SubtractVectorFromPosition(Position=Linked() {linkedVariable {name=System.Position, type=Vector3f}}, Vector=<0 -2 0>)
    Sprite Size Mode=::NewEnumerator3
    Uniform Sprite Size=70
    UsePositionOffset=true
    Write Material Random=true
   )
   &Color(Color=&MakeLinearColorFromVectorAndFloat("Vector (RGB)"=<13 0 9.9273>, "Float (Alpha)"=0.1))
   &CameraOffset(Camera Offset Amount=5)
  ]
  particleUpdateScript [
   &ParticleState(Kill Particles When Lifetime Has Elapsed=false, Loop Particles Lifetime=true)
   &ScaleColor(Scale Alpha=&FN_FrontEndFloatModifier(Gameplay=0.8, Front End=0.4), ScaleA=true, ScaleRGB=true)
   &FN_Light_Attributes(
    Color=$Particles.Color
    Create Color Attribute=true
    Create Exponent Attribute=true
    Create Radius Attribute=true
    Exponent=2
    Radius=21
    Scale Color=2
   )
  ]
  rendererClasses [&NiagaraSpriteRendererProperties, &NiagaraLightRendererProperties]
  renderers [
   {rendererClass=&NiagaraSpriteRendererProperties}
   {rendererClass=&NiagaraLightRendererProperties, rendererIndex=1}
  ]
 }
 Symbol {
  emitterUpdateScript [
   &EmitterState(
    Inactive Response=::NewEnumerator1
    Life Cycle Mode=::NewEnumerator1
    Loop Behavior=::NewEnumerator1
    Loop Duration Mode=::NewEnumerator1
    UseLoopDelay=false
   )
   &SpawnBurst_Instantaneous(Use Spawn Probability=false, UseLoopCountLimit=false)
  ]
  particleSpawnScript [
   &InitializeParticle(
    Lifetime=4
    Mesh Scale=&Multiply_VectorByFloat(Vector=<0.1 0.14 0.14>, Float=0.733)
    Mesh Scale Mode=::NewEnumerator4
    Position Offset=&Subtract_Vector(A=$System.Position, B=<0 -2 0>)
    UsePositionOffset=true
    Write Material Random=true
   )
   &Color(Color=&MakeLinearColorFromVectorAndFloat("Vector (RGB)"=<13 0 9.9273>, "Float (Alpha)"=1))
   &InitialMeshOrientation(Mesh Orientation Mode=::NewEnumerator2, Rotation=<0 0 -0.25>, Use Rotation Vector=true)
  ]
  particleUpdateScript [
   &ParticleState(Kill Particles When Lifetime Has Elapsed=false, Loop Particles Lifetime=true)
   &FN_Mesh_Attributes(
    Camera Offset=8
    Create Camera Offset Attribute=true
    Create Position Attribute=true
    Create Scale Attribute=true
    Module Output=::NewEnumerator1
    Position:Vector3f=$System.Position
    Scale=&VectorFromFloat(Value=0.149)
   )
   &FN_Mesh_Attributes(
    Create Position Attribute=true
    Create Scale Attribute=true
    Module Output=::NewEnumerator1
    Position:NiagaraPosition=&Subtract_Vector(A=$System.Position, B=<0 -3 0>)
    Scale=&VectorFromFloat(Value=0.09)
   ) {module=FN_Mesh_Attributes001}
   &FN_Mesh_Attributes(
    Create Position Attribute=true
    Create Scale Attribute=true
    Module Output=::NewEnumerator1
    Position:NiagaraPosition=&AddVectorToPosition(Position=<0.136 3 0>, Vector=$System.Position)
    Scale=&VectorFromFloat(Value=0.012)
   ) {module=FN_Mesh_Attributes002}
  ]
  rendererClasses [&NiagaraMeshRendererProperties]
  renderers [
   {rendererClass=&NiagaraMeshRendererProperties, rendererIndex=0}
   {rendererClass=&NiagaraMeshRendererProperties, rendererIndex=1}
   {rendererClass=&NiagaraMeshRendererProperties, rendererIndex=2}
   {rendererClass=&NiagaraMeshRendererProperties, rendererIndex=3}
  ]
 }
 TriangleInner {
  emitterUpdateScript [
   &EmitterState(
    Inactive Response=::NewEnumerator1
    Life Cycle Mode=::NewEnumerator1
    Loop Behavior=::NewEnumerator1
    Loop Duration Mode=::NewEnumerator1
    UseLoopDelay=false
   )
   &SpawnBurst_Instantaneous(Spawn Count=3, Use Spawn Probability=false, UseLoopCountLimit=false)
  ]
  particleSpawnScript [
   &PartitionParticles(Number Of Partitions=3)
   &InitializeParticle(
    Lifetime=4
    Mesh Scale=&SelectVectorFromArray(
     Direct Array Index=$Particles.PartitionParticles.Partition
     Vector Selection Array=DataInterface() {propertyValues {FloatData [<1.026 0.5 0.5>, <0.9279 0.5 0.5>, <0.9279 0.5 0.5>]}}
    )
    Mesh Scale Mode=::NewEnumerator4
    Position Offset=&AddVectorToPosition(
     Position=&SelectVectorFromArray(
      Direct Array Index=$Particles.PartitionParticles.Partition
      Vector Selection Array=DataInterface() {propertyValues {FloatData [<-4.9959 0 2.7608>, <-5.1501 0 2.5>, <5.4384 0 2.5>]}}
     )
     Vector=&SubtractVectorFromPosition(Position=Linked() {linkedVariable {name=System.Position, type=Vector3f}}, Vector=<0 -3 0>)
    )
    UsePositionOffset=true
    Write Material Random=true
   )
   &Color(Color=&MakeLinearColorFromVectorAndFloat("Vector (RGB)"=<13 0 9.9273>, "Float (Alpha)"=0.1))
   &DynamicMaterialParameters(
    Index 0 Param 1=0.4
    Index 0 Param 2=1
    Index 0 Param 3=1
    Index 0 Param 4=1
    Module.Param0WriteEnabled=true
    Module.Param1WriteEnabled=true
    Module.Param2WriteEnabled=true
    Module.Param3WriteEnabled=true
   )
   &InitialMeshOrientation(
    Mesh Orientation Mode=::NewEnumerator2
    Rotation=&SelectVectorFromArray(
     Direct Array Index=$Particles.PartitionParticles.Partition
     Vector Selection Array=DataInterface() {propertyValues {FloatData [<0 0 0>, <0 0.15 0>, <0 0.35 0>]}}
    )
    Use Rotation Vector=true
   )
  ]
  particleUpdateScript [&ParticleState(Kill Particles When Lifetime Has Elapsed=false, Loop Particles Lifetime=true)]
  rendererClasses [&NiagaraMeshRendererProperties]
  renderers [{rendererClass=&NiagaraMeshRendererProperties, rendererIndex=0}]
 }
 TriangleOutter {
  emitterUpdateScript [
   &EmitterState(
    Inactive Response=::NewEnumerator1
    Life Cycle Mode=::NewEnumerator1
    Loop Behavior=::NewEnumerator1
    Loop Duration Mode=::NewEnumerator1
    UseLoopDelay=false
   )
   &SpawnBurst_Instantaneous(Spawn Count=3, Use Spawn Probability=false, UseLoopCountLimit=false)
  ]
  particleSpawnScript [
   &PartitionParticles(Number Of Partitions=3)
   &InitializeParticle(
    Lifetime=4
    Mesh Scale=&Multiply_VectorByFloat(Vector=<2.32 0.5 0.5>, Float=0.733)
    Mesh Scale Mode=::NewEnumerator4
    Position Offset=&AddVectorToPosition(
     Position=&SelectVectorFromArray(
      Direct Array Index=$Particles.PartitionParticles.Partition
      Vector Selection Array=DataInterface() {propertyValues {FloatData [<-8.5 0 -4.9>, <-8.5 0 -5>, <8.5 0 -5>]}}
     )
     Vector=&SubtractVectorFromPosition(Position=Linked() {linkedVariable {name=System.Position, type=Vector3f}}, Vector=<0 -2 0>)
    )
    UsePositionOffset=true
    Write Material Random=true
   )
   &Color(Color=&MakeLinearColorFromVectorAndFloat("Vector (RGB)"=<13 0 9.9273>, "Float (Alpha)"=1))
   &InitialMeshOrientation(
    Mesh Orientation Mode=::NewEnumerator2
    Rotation=&SelectVectorFromArray(
     Direct Array Index=$Particles.PartitionParticles.Partition
     Vector Selection Array=DataInterface() {propertyValues {FloatData [<0 0 0>, <0 -0.166 0>, <0 0.666 0>]}}
    )
    Use Rotation Vector=true
   )
   &DynamicMaterialParameters(
    Index 0 Param 1=0.4
    Index 0 Param 2=1
    Index 0 Param 3=1
    Index 0 Param 4=1
    Module.Param0WriteEnabled=true
    Module.Param1WriteEnabled=true
    Module.Param2WriteEnabled=true
    Module.Param3WriteEnabled=true
   )
  ]
  particleUpdateScript [
   &ParticleState(Kill Particles When Lifetime Has Elapsed=false, Loop Particles Lifetime=true)
   &ScaleColor(
    Scale RGB=&VectorFromCurve(VectorCurve=DataInterface() {propertyValues {XCurve {keys [<at 0 1>,<at 1 1>]}, YCurve {keys [<at 0 1>,<at 1 1>]}, ZCurve {keys [<at 0 1>,<at 1 1>]}}})
    ScaleA=true
    ScaleRGB=true
   )
  ]
  rendererClasses [&NiagaraMeshRendererProperties]
  renderers [{rendererClass=&NiagaraMeshRendererProperties, rendererIndex=0}]
 }
}
summary {
 systemName=NS_Backpack_GoatRobe
 emitters [
  {emitterName=Light, rendererClasses [&NiagaraSpriteRendererProperties, &NiagaraLightRendererProperties]}
  {emitterName=Symbol, rendererClasses [&NiagaraMeshRendererProperties]}
  {emitterName=TriangleOutter, rendererClasses [&NiagaraMeshRendererProperties]}
  {emitterName=TriangleInner, rendererClasses [&NiagaraMeshRendererProperties]}
 ]
 userVariables [
  User.bIsFrontEnd {type=NiagaraBool}
  User.bIsFrontEndPreview {type=NiagaraBool}
 ]
}
